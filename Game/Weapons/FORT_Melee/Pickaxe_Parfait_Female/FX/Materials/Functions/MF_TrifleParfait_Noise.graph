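# Material function for the Parfait pickaxe FX that produces animated lightning-style noise as a World Position Offset and a width multiplier.
# Three panned samples of a lightning noise texture along UV V are weighted, summed and faded at both ends by a sine mask.
# Displacement follows a camera-relative direction, rotated around the vertex tangent by a fourth noise sample.
# A PreviousFrameSwitch exaggerates motion blur, and width narrows where noise is weak.
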MaterialFunction /Game/Weapons/FORT_Melee/Pickaxe_Parfait_Female/FX/Materials/Functions/MF_TrifleParfait_Noise
graph {nodes [
  n1 = Constant()
  n2 = ComponentMask(None=TextureCoordinate()) {g=true}
  n3 = FunctionInput() {bUsePreviewValueAsDefault=true, inputName=Noise UV Scale, inputType=FunctionInput_Scalar, previewValue=<1 0 0 1>}
  n4 = Multiply(A=@n2, B=@n3)
  n5 = VectorParameter() {parameterName=UV 1 Scale and Speeds, defaultValue=<rgba 0.333 0.01 0.2 0>}
  n6 = Multiply(A=@n4, B=@n5.R)
  n7 = FunctionInput(Preview=Time()) {bUsePreviewValueAsDefault=true, inputName=Time, inputType=FunctionInput_Scalar}
  n8 = TextureSample(UVs=Panner(Coordinate=AppendVector(A=@n1, B=@n6), Time=@n7, Speed=AppendVector(A=@n5.G, B=@n5.G))) {
   samplerSource=SSM_Wrap_WorldGroupSettings
   samplerType=SAMPLERTYPE_LinearGrayscale
   texture=/Game/Weapons/FORT_Melee/Pickaxe_Parfait_Female/FX/Textures/T_TrifleParfait_LightningNoise
  }
  n9 = ConstantBiasScale(None=@n8.R) {bias=-0.5, scale=2}
  n10 = ScalarParameter() {defaultValue=6, parameterName=Noise 1 Intensity}
  n11 = Constant()
  n12 = VectorParameter() {parameterName=UV 2 Scale and Speeds, defaultValue=<rgba 1 -0.0333 -0.5 0>}
  n13 = Multiply(A=@n4, B=@n12.R)
  n14 = TextureSample(UVs=Panner(Coordinate=AppendVector(A=@n11, B=@n13), Time=@n7, Speed=AppendVector(A=@n12.G, B=@n12.G))) {
   samplerSource=SSM_Wrap_WorldGroupSettings
   samplerType=SAMPLERTYPE_LinearGrayscale
   texture=&T_TrifleParfait_LightningNoise
  }
  n15 = ConstantBiasScale(None=@n14.R) {bias=-0.5, scale=2}
  n16 = ScalarParameter() {defaultValue=2, parameterName=Noise 2 Intensity}
  n17 = Constant()
  n18 = VectorParameter() {parameterName=UV 3 Scale and Speeds, defaultValue=<rgba 4.23 0.45 0.3 0>}
  n19 = Multiply(A=@n4, B=@n18.R)
  n20 = TextureSample(UVs=Panner(Coordinate=AppendVector(A=@n17, B=@n19), Time=@n7, Speed=AppendVector(A=@n18.G, B=@n18.G))) {
   samplerSource=SSM_Wrap_WorldGroupSettings
   samplerType=SAMPLERTYPE_LinearGrayscale
   texture=&T_TrifleParfait_LightningNoise
  }
  n21 = ConstantBiasScale(None=@n20.R) {bias=-0.5, scale=2}
  n22 = ScalarParameter() {defaultValue=1, parameterName=Noise 3 Intensity}
  n23 = Add(A=Add(A=Multiply(A=@n9, B=@n10), B=Multiply(A=@n15, B=@n16)), B=Multiply(A=@n21, B=@n22))
  n24 = FunctionInput(Preview=ScalarParameter() {defaultValue=0.5, parameterName=Noise Global Intensity}) {bUsePreviewValueAsDefault=true, inputName=Noise Intensity, inputType=FunctionInput_Scalar, previewValue=<1 0 0 1>}
  n25 = Subtract(A=CameraPositionWS(), B=WorldPosition().XYZ)
  n26 = VertexNormalWS()
  n27 = VertexTangentWS()
  n28 = Normalize(VectorInput=CrossProduct(A=Normalize(VectorInput=@n25), B=CrossProduct(A=@n26, B=@n27)))
  n29 = Constant()
  n30 = VectorParameter() {parameterName="UV Rotation Scale, Speeds, Range", defaultValue=<rgba 0.4 -0.006 0.3 0.25>}
  n31 = Multiply(A=@n4, B=@n30.R)
  n32 = TextureSample(UVs=Panner(Coordinate=AppendVector(A=@n29, B=@n31), Time=@n7, Speed=AppendVector(A=@n30.G, B=@n30.G))) {
   samplerSource=SSM_Wrap_WorldGroupSettings
   samplerType=SAMPLERTYPE_LinearGrayscale
   texture=&T_TrifleParfait_LightningNoise
  }
  n33 = ConstantBiasScale(None=@n32.R) {bias=-0.5, scale=2}
  n34 = RotateAboutAxis(
   NormalizedRotationAxis=@n27
   RotationAngle=FunctionInput(Preview=Multiply(A=@n33, B=@n30.A)) {bUsePreviewValueAsDefault=true, inputName=Noise Axis Rotation, inputType=FunctionInput_Scalar, sortPriority=1}
   PivotPoint=Constant()
   Position=@n28
  )
  n35 = Multiply(A=Multiply(A=Multiply(A=@n23, B=@n24), B=NamedRerouteUsage()), B=Normalize(VectorInput=Add(A=@n28, B=@n34)))
  n36 = Multiply(A=@n35, B=ScalarParameter() {parameterName=Noise Motion Blur Intensity})
  FunctionOutput(None=PreviousFrameSwitch(Current Frame=@n35, Previous Frame=Add(A=@n35, B=@n36))) {outputName=Noise WPO}
  n37 = Abs(None=@n23)
  n38 = Add(A=Add(A=@n10, B=@n16), B=@n22)
  n39 = OneMinus(None=Saturate(None=Divide(A=@n37, B=@n38)))
  n40 = ScalarParameter() {defaultValue=0.5, parameterName=Noise Width Squeeze}
  FunctionOutput(None=OneMinus(None=Multiply(A=@n39, B=@n40))) {outputName=Noise Width Multiplier, sortPriority=1}
  MaterialFunctionCall!(Base=@n9, Exponent) {materialFunction=&/MF_PowerSign}
  MaterialFunctionCall(Base=@n15) {materialFunction=&MF_PowerSign}
  MaterialFunctionCall(Base=@n21) {materialFunction=&MF_PowerSign}
  NamedRerouteDeclaration(None=Sine(None=Divide(A=@n2))) {name=Noise Mask, nodeColor=<rgba 0.2422 1 0 1>}
 ]}
