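# Material function for the Parfait female pickaxe FX that thickens spline-like geometry through a view-dependent World Position Offset.
# Crosses the camera vector with the normal-tangent cross product, then offsets vertices either side of U=0.5, tapering width at both V ends.
# Width is clamped to a minimum screen-pixel size computed from view depth, FOV and viewport width.
# A second output gives the saturated requested-to-minimum width ratio, for dimming widened distant geometry.
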MaterialFunction /Game/Weapons/FORT_Melee/Pickaxe_Parfait_Female/FX/Materials/Functions/MF_TrifleParfait_SplineThicken
graph {nodes [
  n1 = CameraPositionWS()
  n2 = WorldPosition()
  n3 = CrossProduct(A=Normalize(VectorInput=Subtract(A=@n1, B=@n2.XYZ)), B=CrossProduct(A=VertexNormalWS(), B=VertexTangentWS()))
  n4 = Subtract(A=ComponentMask(None=TextureCoordinate()) {r=true}) {constB=0.5}
  n5 = FunctionInput() {bUsePreviewValueAsDefault=true, inputName=Width, inputType=FunctionInput_Scalar, previewValue=<1 0 0 1>}
  n6 = Multiply(A=ComponentMask(None=ViewProperty() {property=MEVP_TanHalfFieldOfView}.Property) {r=true}) {constB=2}
  n7 = DotProduct(
   A=Subtract(A=WorldPosition() {worldPositionShaderOffset=WPT_ExcludeAllShaderOffsets}.XYZ, B=CameraPositionWS())
   B=MaterialFunctionCall() {materialFunction=&CameraDirectionVector}.Result
  )
  n8 = ViewSize()
  n9 = Multiply(A=ScalarParameter() {defaultValue=3, parameterName=Width in Pixels}, B=Divide(A=Multiply(A=@n6, B=@n7), B=ComponentMask(None=@n8) {r=true}))
  n10 = ComponentMask(None=TextureCoordinate()) {g=true}
  n11 = OneMinus(None=Abs(None=ConstantBiasScale(None=@n10) {bias=-0.5, scale=2}))
  n12 = ScalarParameter() {defaultValue=0.1, parameterName=Width Mask Size}
  n13 = Multiply(A=Max(A=@n5, B=@n9), B=Saturate(None=Divide(A=@n11, B=@n12)))
  FunctionOutput(None=Multiply(A=Normalize(VectorInput=@n3), B=Multiply(A=@n4, B=@n13))) {outputName=WPO}
  FunctionOutput(None=Saturate(None=Divide(A=@n5, B=@n9))) {outputName=Distance Intensity Modulation, sortPriority=1}
 ]}
