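# Masked, unlit, two-sided mesh-particle material for lightning effects on the Parfait female pickaxe.
# Noise displacement and spline thickening form World Position Offset, with width shaped by brightness oscillation and a sliding gradient mask, optionally flipped per particle.
# Dynamic parameters drive width, mask and noise, with preview fallbacks when particle relative time is zero.
# Vertex-interpolated color pulses with oscillation and distance intensity, while opacity blends noise masking with a smoke-trail texture.
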Material /Game/Weapons/FORT_Melee/Pickaxe_Parfait_Female/FX/Materials/M_Pickaxe_Parfait_Lightning
graph {
 properties {
  bUsedWithNiagaraMeshParticles=true
  blendMode=BLEND_Masked
  pixelDepthOffsetMode=PDOM_Legacy
  shadingModel=MSM_Unlit
  twoSided=true
  usedShadingModels=Unlit
 }
 nodes [
  n1 = DynamicParameter() {paramNames [Width, Mask Size, Mask Position, Noise Intensity]}
  n2 = MaterialFunctionCall!(
   Noise Intensity=@n1.Noise Intensity
   Noise UV Scale=ScalarParameter() {defaultValue=1, parameterName=Noise UV Scale}
   Time
   Noise Axis Rotation
  ) {materialFunction=/Game/Weapons/FORT_Melee/Pickaxe_Parfait_Female/FX/Materials/Functions/MF_TrifleParfait_Noise}
  n3 = MaterialFunctionCall!(In) {materialFunction=/Game/Weapons/FORT_Melee/Pickaxe_Parfait_Female/FX/Materials/Functions/MF_TrifleParfait_BrightnessOscillation}
  n4 = Multiply(A=@n2.Noise Width Multiplier, B=@n3.Brightness)
  n5 = Ceil(None=ParticleRelativeTime() {desc=">0"})
  n6 = LinearInterpolate(A=Constant() {r=1.333}, B=@n1.Width, Alpha=@n5)
  n7 = ComponentMask(None=TextureCoordinate()) {g=true}
  n8 = OneMinus(None=@n7)
  n9 = Round(None=ParticleRandom())
  n10 = Time()
  n11 = MaterialFunctionCall!(
   Gradient=LinearInterpolate(A=@n7, B=LinearInterpolate(A=@n7, B=@n8, Alpha=@n9), Alpha=ScalarParameter() {defaultValue=1, parameterName=UseParticleRandomFlip})
   Mask Position=LinearInterpolate(A=Frac(None=@n10), B=@n1.Mask Position, Alpha=@n5) {constA=0.5}
   Mask Size=LinearInterpolate(A=Constant() {r=3}, B=@n1.Mask Size, Alpha=@n5)
  ) {materialFunction=/Game/Effects/Fort_Effects/MaterialFunctions/MF_GradientMask}
  n12 = MaterialFunctionCall!(Width=Multiply(A=Multiply(A=@n4, B=@n6), B=@n11.Sinusoidal)) {materialFunction=/Game/Weapons/FORT_Melee/Pickaxe_Parfait_Female/FX/Materials/Functions/MF_TrifleParfait_SplineThicken}
  Add(A=@n2.Noise WPO, B=@n12.WPO)
  n13 = ParticleColor()
  n14 = VectorParameter() {group=Opacity, parameterName=SmokeTrail UV, sortPriority=1, defaultValue=<rgba 1 1 0 0.2>}
  n15 = Multiply(A=TextureCoordinate(), B=AppendVector(A=@n14.R, B=@n14.R))
  n16 = AppendVector(A=@n14.B, B=@n14.B)
  n17 = TextureSampleParameter2D(UVs=MaterialFunctionCall(
    UVs=Panner(Coordinate=@n15, Speed=@n16)
    "Rotation Angle (0-1)"=ScalarParameter() {defaultValue=0.25, group=Opacity, parameterName=SmokeTrailTextureRotation}
   ) {materialFunction=&CustomRotator}.Rotated Values) {group=Opacity, parameterName=SmokeTrailTexture, texture=/Game/Effects/Fort_Effects/Textures/Smoke/T_Smoke_Trail_Soft_Chromatic}
  n18 = Power(Base=@n12.Distance Intensity Modulation, Exp=ScalarParameter() {defaultValue=1.25, parameterName=Distance Intensity Modulation Exponent})
  n19 = Multiply(
   A=MaterialFunctionCall!(Noise Intensity=VertexInterpolator(VS=@n18).PS) {materialFunction=/Game/Weapons/FORT_Melee/Pickaxe_Parfait_Female/FX/Materials/Functions/MF_TrifleParfait_OpacityNoise}.Masked Opacity
   B=@n11.Sinusoidal
  )
  n20 = ParticleColor()
  n21 = NamedRerouteUsage()
  LinearInterpolate(
   A=LinearInterpolate(A=Multiply(A=@n13.A, B=@n17.R), B=@n19, Alpha=ScalarParameter() {parameterName=Tanslucent Masked})
   B=Multiply(A=@n19, B=LinearInterpolate(B=@n20.A, Alpha=@n21) {constA=1})
   Alpha=ScalarParameter() {parameterName=UseOpacityMaskInputForOpacity}
  )
  n22 = LinearInterpolate(
   A=VectorParameter() {parameterName=Default Color, defaultValue=<rgba 0 27.7494 333 3>}.RGB
   B=ParticleColor().RGB
   Alpha=Ceil(None=ParticleRelativeTime())
  )
  n23 = Power(Base=@n3.Brightness, Exp=ScalarParameter() {defaultValue=3, parameterName=Brightness Oscillation Exponent})
  n24 = Max(A=VectorParameter() {parameterName=Minimum Color, defaultValue=<rgba 0.15 0.49 1 0>}.RGB, B=Multiply(A=Multiply(A=@n22, B=@n23), B=@n18))
  n25 = Multiply(A=Multiply(A=ParticleColor().RGB, B=@n23), B=@n18)
  n26 = NamedRerouteUsage()
  MaterialFunctionCall(Emissive=VertexInterpolator(VS=LinearInterpolate(A=@n24, B=@n25, Alpha=@n26)).PS) {materialFunction=&MF_FX_EmissiveAdjustments}
  n27 = ScalarParameter() {parameterName=Use Particle COlor For COlor and alpha}
  NamedRerouteDeclaration(None=@n27) {name=PARAM Use Particle COlor For COlor and alpha, nodeColor=<rgba 0 1 0.1289 1>}
  Reroute(None=@n27)
  ScalarParameter() {parameterName=Use Particle RElative time}
 ]
}
