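# Physics asset for the front-end mesh of the Parfait Female two-handed pickaxe, giving its rope bones physical bodies.
# A long capsule covers the handle, and three short capsules form a rope chain, each constrained to the body before it.
# The rope root has a tighter swing limit against the handle than the two looser lower joints, and all three joints set twist limits to zero.
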PhysicsAsset /Game/Weapons/FORT_Melee/Pickaxe_Parfait_Female/Meshes/Parfait_Female_Axe_FrontEnd_PhysicsAsset
bodies {
 Melee_TwoHanded_Handle [{shapeType=Capsule, length=28.25, radius=2, center=<0 0 0.2019>}]
 rope [{shapeType=Capsule, length=1.3, radius=1.35, shapeName=rope_capsule, center=<1.9539 0 0>, rotation=<rot 90 0 0>}]
 rope_01 [{shapeType=Capsule, length=1.5, radius=1.25, shapeName=rope_01_capsule, center=<1.3333 0 0>, rotation=<rot 90 0 0>}]
 rope_02 [{shapeType=Capsule, length=1.5, radius=1.25, shapeName=rope_02_capsule, center=<1.8554 0 0>, rotation=<rot 90 0 0>}]
}
constraints [
 {bone1Name=rope, bone2Name=Melee_TwoHanded_Handle, swing1LimitDegrees=45, swing2LimitDegrees=45, twistLimitDegrees=0}
 {bone1Name=rope_01, bone2Name=rope, swing1LimitDegrees=100, swing2LimitDegrees=100, twistLimitDegrees=0}
 {bone1Name=rope_02, bone2Name=rope_01, swing1LimitDegrees=100, swing2LimitDegrees=100, twistLimitDegrees=0}
]
